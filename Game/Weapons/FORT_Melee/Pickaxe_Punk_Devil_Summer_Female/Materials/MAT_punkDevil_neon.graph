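# MAT_punkDevil_neon is a Default Lit material for the neon look of the female summer Punk Devil pickaxe, usable on skeletal meshes and mesh deformers.
# It uses material attributes, taking the result of the MF_punkDevil_neon function and feeding it into the shared MF_TechArtOverrides_accessories function.
# The graph holds only these two function calls, so the actual shading logic lives inside those functions.
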Material /Game/Weapons/FORT_Melee/Pickaxe_Punk_Devil_Summer_Female/Materials/MAT_punkDevil_neon
graph {
 properties {
  bUseMaterialAttributes=true
  bUsedWithMeshDeformer=true
  bUsedWithSkeletalMesh=true
  pixelDepthOffsetMode=PDOM_Legacy
  usedShadingModels=Default Lit
 }
 nodes [MaterialFunctionCall(In=MaterialFunctionCall() {materialFunction=&/MF_punkDevil_neon}.Result) {materialFunction=&MF_TechArtOverrides_accessories}]
}
